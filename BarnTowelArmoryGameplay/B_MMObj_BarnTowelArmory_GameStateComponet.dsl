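; Server-side game state component that spawns BarnTowel and MurkyVole Armory variants in the air at landing locations once the playlist is ready, if enabled.
; It gathers spawn managers from the global actor registry, then starts delayed, chained spawns immediately or at a configured storm phase.
; The storm trigger runs EQS to order locations and cap the count, with a deterministic override.
; Hotfix values select fixed-sequence, forced-count or weighted-random variants.
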(event EventBeginPlay
  ;[0 1 0.004913] BeginPlay
  ; If enabled, register the landing locations and set up the spawning events
  (|CallParentFunction)
  (Utilities|SwitchHasAuthority
    (:Authority
      (Class|FortAsyncActionCurrentPlaylistReady|PlaylistDataReadyAsync
        (:OnReady
          (if (ScalableFloat|GetValueasBool (Variables|BalanceRows|GetHFRow_ArmorySpawningEnabled))
            (CallFunction|InitArmoryVariables)
            (Registration|GetandRegisterLandingLocations)
            (CallFunction|InitSpawningSetup)
            (CallFunction|InitArmoryVariantForceSpawnVariables)))))))

(event Custom|InitSpawningSetup
  ;[0 1 0.004913] InitSpawningSetup
  ; Setup the spawning to be based on the Storm Circle timing or just generic timers
  (if (ScalableFloat|GetValueasBool (Variables|BalanceRows|GetHFRow_EnableStormTimedSpawn))
    ;;[0.3135 0.4972 0.7309] Bind to game phase step changed to listen for new storm circles
    (bind _returnvalue (GameplayEventRouter|GetEventRouter "/Script/FortniteGame.FortGameState" (Components|GetOwner)))
    (bind (_asynctaskproxy _payload) (GameplayEventRouter|StartListeningToStatefulEvent _returnvalue "/Script/FortniteGame.FortGamePhaseStepUpdatedEvent")
      (:then
        (Variables|Spawning|SetGamePhaseStepUpdatedAsyncTask _asynctaskproxy))
      (:OnEventReceived
        _)
      (:OnSavedEventStateExists
        _))
    (CallFunction|OnGamePhaseStepChanged (Utilities|Struct|BreakFortGamePhaseStepUpdatedEvent _payload))
    (else
      (CallFunction|BeginDelayedSpawn))))

(event Custom|OnGamePhaseStepChanged
  (params (NewGamePhaseStep :type EAthenaGamePhaseStep%20Enum))
  ;[0.3976 0.2032 0.1242] OnGamePhaseStepChanged
  ; Checks the storm phase for spawning timing, if enabled
  (if (== (Math|Conversions|ToInteger(Byte) NewGamePhaseStep) (ScalableFloat|GetValueasInteger (Variables|BalanceRows|GetHFRow_StormPhaseStepToSpawn)))
    (if (== (SafeZone|GetCurrentPhaseIndex (SafeZone|GetSafeZoneIndicator)) (ScalableFloat|GetValueasInteger (Variables|BalanceRows|GetHFRow_StormPhaseCountToSpawn)))
      (CallFunction|BeginDelayedSpawn)
      (Spawning|FindBestSpawnLocationsfromEQS))))

(event Custom|BeginDelayedSpawn
  ;[0.3265 1 0.2934] BeginDelayedSpawn
  ; Starts a delayed spawn from the original trigger (either BeginPlay or Storm Timing), unless timer is set to zero or less
  (bind _output_get (Variables|Spawning|SetRolledSpawnDelayTime (Math|Float|SelectFloat (ScalableFloat|GetValueatLevel (Variables|BalanceRows|GetHFRow_DelayInitialSpawnMin)) (Math|Random|RandomFloatinRange (ScalableFloat|GetValueatLevel (Variables|BalanceRows|GetHFRow_DelayInitialSpawnMin)) (ScalableFloat|GetValueatLevel (Variables|BalanceRows|GetHFRow_DelayInitialSpawnMax))) (Class|FortDeterminismConfig|IsUsingArmoryDeterminism)))) ; store the RNG time so it isn't rerolled every time
  (if (> _output_get 0)
    (bind _returnvalue (Utilities|Time|SetTimerbyEvent (EventDispatchers|CreateEvent :!event SelectArmoryFromRegisteredSpawners) _output_get))
    (Variables|Spawning|SetTimer_SpawningDelay _returnvalue)
    (else
      (Spawning|SelectArmoryfromRegisteredSpawners) ; spawn without delay
    )))

(event Custom|SpawnSelectedArmory
  (params (SelectedArmorySpawner :type Actor%20Object%20Reference) (OverrideVariant :type Enum_BarnTowelArmory_SelectedVariant%20Enum :default "NewEnumerator0"))
  ;[0 0.2682 1] SpawnSelectedArmory
  ; Sends spawn event to the Landing Pad to spawn the associated Armory, removes the location from valid spawns so they don't use the same position, and sets up the next spawn timer if the max count isnot yet met
  (bind _asb_barn_towel_spawn_manager (Utilities|Casting|CastToB_BarnTowel_SpawnManager SelectedArmorySpawner))
  (bind _varianttospawn (CallFunction|GetArmoryVarianttoSpawn))
  (Class|BBarnTowelSpawnManager|SpawnArmoryinAir _asb_barn_towel_spawn_manager (Utilities|Select :Index OverrideVariant :NewEnumerator0 _varianttospawn :NewEnumerator2 "NewEnumerator2" :NewEnumerator1 "NewEnumerator1" :NewEnumerator3 _varianttospawn))
  (Utilities|Array|RemoveItem (Variables|Registration|GetAvailableRegisteredLandingLocations) _asb_barn_towel_spawn_manager) ; remove the spawning location as valid, \nprevents Armories from chooing the same \nlocation and landing on top of each other
  (Math|Integer|IncrementInt (Variables|Spawning|GetArmoriesSpawnedCount)
    (:%20%20
      (Utilities|Time|ClearandInvalidateTimerbyHandle (Variables|Spawning|GetTimer_SpawningDelay))
      (if (< (Variables|Spawning|GetArmoriesSpawnedCount) (Variables|Spawning|GetEffectiveMaxArmorySpawns))
        (bind _output_get (Variables|Spawning|SetRolledSpawnDelayTime (Math|Random|RandomFloatinRange (ScalableFloat|GetValueatLevel (Variables|BalanceRows|GetHFRow_DelayAdditionalSpawnMin)) (ScalableFloat|GetValueatLevel (Variables|BalanceRows|GetHFRow_DelayAdditionalSpawnMax)))))
        (if (> _output_get 0)
          (bind _returnvalue (Utilities|Time|SetTimerbyEvent (EventDispatchers|CreateEvent :!event SelectArmoryFromRegisteredSpawners) _output_get))
          (Variables|Spawning|SetTimer_SpawningDelay _returnvalue)
          (else
            (Spawning|SelectArmoryfromRegisteredSpawners) ; spawn without delay
          ))))) ; increment current Armory counter
)

(event Custom|DebugSpawnArmoryInFiringRange
  (Utilities|FlowControl|Delay 0.8)
  ;[1 1 1] There really should be only one Armory Spawner in the firing range, spawn the first one found and then print a warning that this isn't expected
  (bind _outactors (Actor|GetAllActorsOfClass "/BarnTowelArmoryGameplay/Items/Environmental/B_BarnTowel_SpawnManager.B_BarnTowel_SpawnManager_C"))
  (bind _spawnedarmory (Class|BBarnTowelSpawnManager|SpawnArmoryinAir (Utilities|Array|Get(acopy) _outactors) "NewEnumerator0"))
  (Default|BindEventtoOnArmoryStateChanged _spawnedarmory (EventDispatchers|CreateEvent :!event Custom|OnFiringRangeArmoryStateUpdated)) ; WARNING: This is the last event in the Armory state-manager \nsequence that can be accessed currently for firing range.
  (if (> (Utilities|Array|Length _outactors) 1)
    (Development|PrintString "BARNTOWELARMORY: Firing Range has more than one Armory Spawner, but only one is expected! It may not spawn in the expected locationand may have errors." true true "(R=1.000000,G=0.064735,B=0.000000,A=1.000000)" 3e+01)))

(event Custom|OnFiringRangeArmoryStateUpdated
  (params (NewGameplayState :type Enum_BarnTowelArmory_LandingState%20Enum :default "NewEnumerator0"))
  (if (Utilities|Enum|Equal(Enum) NewGameplayState "NewEnumerator4")
    (bind _outactors (Actor|GetAllActorswithTag "BarnTowelArmory.TestLevel.InteriorLoader"))
    (for _array_element _outactors
      (bind _asbuilding_foundation_3x_3 (Utilities|Casting|CastToBuildingFoundation3x3 _array_element))
      (Class|BuildingFoundation|SetDynamicFoundationEnabled _asbuilding_foundation_3x_3 true))
    (bind _outactors_2 (Actor|GetAllActorsOfClass "/BarnTowelArmoryGameplay/Items/Environmental/ArmoryLanding/B_ArmoryLandingPad_Flat.B_ArmoryLandingPad_Flat_C"))
    (for _array_element_2 _outactors_2
      (Actor|DestroyActor _array_element_2))))

(event Custom|OverrideBarnTowelArmoryEnabled
  (params (EnableArmorySpawning :type Boolean))
  (bind _cachedlandingregistrysubsystem (Variables|Registration|GetCachedLandingRegistrySubsystem))
  (if EnableArmorySpawning
    (Registration|GetandRegisterLandingLocations)
    (CallFunction|InitSpawningSetup)
    (else
      (Utilities|Time|ClearandInvalidateTimerbyHandle (Variables|Spawning|GetTimer_SpawningDelay))
      (Utilities|Array|Clear (Variables|Registration|GetAvailableRegisteredLandingLocations))
      (Utilities|IsValid _cachedlandingregistrysubsystem
        (:Is%20Valid
          (EventDispatchers|UnbindEventfromOnActorUnregistered _cachedlandingregistrysubsystem (EventDispatchers|CreateEvent :!event CacheRegisteredLandingLocations)))))))

(event Custom|SpawnArmoryFromDebugIndex
  (params (DebugIndex :type Integer))
  (bind _array_element (Utilities|Array|ForEachLoopwithBreak (Variables|Registration|GetAvailableRegisteredLandingLocations))
    (:LoopBody
      (bind _asb_barn_towel_spawn_manager (Utilities|Casting|CastToB_BarnTowel_SpawnManager _array_element))
      (bind _returnvalue (Utilities|GetDataTableRow (Utilities|Struct|BreakDataTableRowHandle (Variables|/BarnTowelArmoryGameplay/Items/Environmental/B_BarnTowel_SpawnManager.B_BarnTowel_SpawnManager_C|GetSpawnData _asb_barn_towel_spawn_manager)) (.RowName (Utilities|Struct|BreakDataTableRowHandle (Variables|/BarnTowelArmoryGameplay/Items/Environmental/B_BarnTowel_SpawnManager.B_BarnTowel_SpawnManager_C|GetSpawnData _asb_barn_towel_spawn_manager)))))
      (if (== (.BarnTowelVariantLandingINDataLayer_35_2FF476334D20D0D7332405AC4D6D5834 (Utilities|Struct|BreakBarnTowelArmorySpawnLocationInfo _returnvalue)) (Utilities|Map|Find (Variables|Debugging|GetMAP_DebugIndexToDataLayer) DebugIndex))
        (CallFunction|SpawnSelectedArmory _asb_barn_towel_spawn_manager)))))

(event Custom|SpawnArmoryByDebugName
  (params (DebugLocationName :type String) (OptionalVariantOverride :type String))
  (bind _debugoverridevariantmap (Variables|Debugging|GetDebugOverrideVariantMap))
  (bind (_value _returnvalue) (Utilities|Map|Find _debugoverridevariantmap OptionalVariantOverride))
  (bind _returnvalue_2 (select _returnvalue _value "NewEnumerator0"))
  (bind _array_element (Utilities|Array|ForEachLoopwithBreak (Variables|Registration|GetAvailableRegisteredLandingLocations))
    (:LoopBody
      (bind _asb_barn_towel_spawn_manager (Utilities|Casting|CastToB_BarnTowel_SpawnManager _array_element))
      (bind _returnvalue_3 (Utilities|GetDataTableRow (Utilities|Struct|BreakDataTableRowHandle (Variables|/BarnTowelArmoryGameplay/Items/Environmental/B_BarnTowel_SpawnManager.B_BarnTowel_SpawnManager_C|GetSpawnData _asb_barn_towel_spawn_manager)) (.RowName (Utilities|Struct|BreakDataTableRowHandle (Variables|/BarnTowelArmoryGameplay/Items/Environmental/B_BarnTowel_SpawnManager.B_BarnTowel_SpawnManager_C|GetSpawnData _asb_barn_towel_spawn_manager)))))
      (if (== (Utilities|Struct|BreakBarnTowelArmorySpawnLocationInfo _returnvalue_3) DebugLocationName)
        (CallFunction|SpawnSelectedArmory _asb_barn_towel_spawn_manager _returnvalue_2)
        (Development|PrintString (Utilities|String|Append "BARNARMORY: Cheat spawning armory with override variant  = " (Utilities|String|EnumtoString _returnvalue_2)) true true "(R=1.000000,G=0.880715,B=0.000000,A=1.000000)" 16.0)))))

(event Custom|DebugRunArmorySpawnEQS
  (Spawning|FindBestSpawnLocationsfromEQS))

(fn CacheAndSortSelectedSpawnerOrder
  (params (ValidEQSSpawners :type Array%20of%20Actor%20Object%20References) (QueryInstance :type EQS%20Query%20Instance%20Object%20Reference))
  (bind _hreturnsortedarray (LocalVariables|GethReturnSortedArray))
  (Utilities|IsValid QueryInstance
    (:Is%20Valid
      (for (_array_element _array_index) ValidEQSSpawners
        (LocalVariables|SethHasAddedElement false)
        (bind (_array_element_2 _array_index_2) (Utilities|Array|ForEachLoopwithBreak _hreturnsortedarray)
          (:LoopBody
            (if (>= (AI|EQS|GetItemScore QueryInstance _array_index) (AI|EQS|GetItemScore QueryInstance _array_index_2))
              (Utilities|Array|Insert _hreturnsortedarray _array_element _array_index_2)
              (LocalVariables|SethHasAddedElement true)))
          (:Completed
            _))
        (if (not (LocalVariables|GethHasAddedElement))
          (Utilities|Array|Add _hreturnsortedarray _array_element) ; put the item at the end of the array
        ))
      ;[0.9327 0.2914 1] RETURN: Sorted array of valid EQS results for spawners
      (return (LocalVariables|GethReturnSortedArray)))
    (:Is%20Not%20Valid
      ;[0.9327 0.2914 1] RETURN: Invalid Array
      (return 0))))

(fn CacheRegisteredLandingLocations
  (params (RegistrationTag :type Gameplay%20Tag%20Structure) (ActorToRegister :type Actor%20Object%20Reference))
  (Utilities|Array|AddUnique (Variables|Registration|GetAvailableRegisteredLandingLocations) ActorToRegister))

(fn FindBestSpawnLocationsFromEQS
  (bind _returnvalue (AI|EQS|RunEQSQuery (select (Class|FortDeterminismConfig|IsUsingArmoryDeterminism) "/BarnTowelArmoryGameplay/Items/EQS/EQS_BarnTowelArmory_Spawner_Deterministic.EQS_BarnTowelArmory_Spawner_Deterministic" "/BarnTowelArmoryGameplay/Items/EQS/EQS_BarnTowelArmory_Spawner.EQS_BarnTowelArmory_Spawner") (Fortnite|Athena|GameState|GetGameStateAthena) "AllMatching"))
  (Utilities|IsValid _returnvalue
    (:Is%20Valid
      (AI|EQS|BindEventtoOnQueryFinished _returnvalue (EventDispatchers|CreateEvent :!event SetSpawnerAfterEQSQueryComplete)))))

(fn GetAndRegisterLandingLocations
  (bind _returnvalue (WorldSubsystems|GetFortGlobalActorRegistrySubsystem))
  (bind _resiteredlandinglocationstag (Variables|Registration|GetResiteredLandingLocationsTag))
  ;[0.4973 0.715 1] GetAndRegisterLandingLocations
  ; Find all the landing locations for this manager to later decide what to activate
  (bind _outactors (Class|FortGlobalActorRegistrySubsystem|GetRegisteredActors _returnvalue _resiteredlandinglocationstag)) ; get all the current registered actors
  (for _array_element _outactors
    ;;[0.4973 0.715 1] Add all found actors - run additional validation if necessary
    (Registration|CacheRegisteredLandingLocations _resiteredlandinglocationstag _array_element))
  ;;[0.4973 0.715 1] Bind to Data Registry object - Will register any defenses that are added to the registry object after the initial run
  (bind _returnvalue_2 (Class|FortGlobalActorRegistrySubsystem|GetActorRegistryDataforGameplayTag _returnvalue _resiteredlandinglocationstag))
  (Utilities|IsValid _returnvalue_2
    (:Is%20Valid
      (bind _output_get (Variables|Registration|SetCachedLandingRegistrySubsystem _returnvalue_2))
      (EventDispatchers|BindEventtoOnActorRegistered _output_get (EventDispatchers|CreateEvent :!event CacheRegisteredLandingLocations)))))

(fn GetArmoryVariantToSpawn
  ;[0.4184 0.7107 1] GetArmoryVariantToSpawn
  ; Determine the type of spawning and get the appropriate variant based on the hotfixable values
  (LocalVariables|SetSelectedForceSpawn "NewEnumerator2") ; use BarnTowel as the default
  (Utilities|FlowControl|Sequence
    (:then_0
      (if (ScalableFloat|GetValueasBool (Variables|BalanceRows|GetHFRow_SpawnVariantUsesFixedSequence))
        ;;[0.4184 0.7107 1] FIXED SEQUENCE SPAWNING
        ;; Spawning using a fixed sequence forces even distribution and each variant takes turns being selected.
        ;; The greater ForcedSpawnNumbers determines which is spawned first, with ties going to BarnTowel
        ;;;[0.9319 0.3559 1] RETURN the next variant in sequence
        (return (select (xor (Math|Conversions|ToBoolean(Integer) (Math|Integer|%25%28Integer%29 (Variables|Spawning|GetArmoriesSpawnedCount) 2)) (>= (Variables|Spawning|GetRemainingForcedTargetBarnTowelToSpawn) (Variables|Spawning|GetRemainingForcedTargetMurkyVoleToSpawn))) "NewEnumerator2" "NewEnumerator1"))
        (elif (or (> (Variables|Spawning|GetRemainingForcedTargetBarnTowelToSpawn) 0) (> (Variables|Spawning|GetRemainingForcedTargetMurkyVoleToSpawn) 0))
          ;;[0 0.003642 0.15 0.5] FORCED SPAWNING
          ;; Guarentees a specific number of each variant based on values set up in SetupArmoryVariantSpawnProperties and the hotfix rows
          ;;;[0.6109 0.8188 1] Only need one of the variants?
          (if (or (<= (Variables|Spawning|GetRemainingForcedTargetBarnTowelToSpawn) 0) (<= (Variables|Spawning|GetRemainingForcedTargetMurkyVoleToSpawn) 0))
            ;;;[0.4184 0.7107 1] select the only needed variant until the remaining force spawn value is completed
            (LocalVariables|SetSelectedForceSpawn (select (>= (Variables|Spawning|GetRemainingForcedTargetBarnTowelToSpawn) (Variables|Spawning|GetRemainingForcedTargetMurkyVoleToSpawn)) "NewEnumerator2" "NewEnumerator1")) ; one of the forced spawn values is not being used, so \nspawn the needed one until all force spawns are completed
            (else
              ;;;[0.4184 0.7107 1] Spawn one of the needed Forced spawned variants, which one doesn't really matter since they will all spawn, but the spawn order will vary based on this roll
              (LocalVariables|SetSelectedForceSpawn (select (<= (Math|Random|RandomIntegerinRange 1 (+ (Variables|Spawning|GetRemainingForcedTargetBarnTowelToSpawn) (Variables|Spawning|GetRemainingForcedTargetMurkyVoleToSpawn))) (Variables|Spawning|GetRemainingForcedTargetBarnTowelToSpawn)) "NewEnumerator2" "NewEnumerator1")) ; one of the forced spawn values is not being used, so \nspawn the needed one until all force spawns are completed
            ))
          (else
            ;;[0.4184 0.7107 1] RNG SPAWNING
            ;; RNG with weights to select the Armory to spawn next
            ;;;[0.9319 0.3559 1] RETURN the random variant
            (return (select (> (Math|Random|RandomFloatinRange 0.0 (+ (ScalableFloat|GetValueatLevel (Variables|BalanceRows|GetHFRow_SpawnWeightBarnTowel)) (ScalableFloat|GetValueatLevel (Variables|BalanceRows|GetHFRow_SpawnWeightMurkyVole)))) (ScalableFloat|GetValueatLevel (Variables|BalanceRows|GetHFRow_SpawnWeightMurkyVole))) "NewEnumerator2" "NewEnumerator1"))))))
    (:then_1
      ;;[0 0.003642 0.15 0.5] Decrement the selected variant when using Forced Spawning
      (switch Utilities|FlowControl|Switch|SwitchonEnum_BarnTowelArmory_SelectedVariant (LocalVariables|GetSelectedForceSpawn)
        (:NewEnumerator0
          (LocalVariables|SetSelectedForceSpawn "NewEnumerator2") ; fallback - this shouldn't happen
        )
        (:NewEnumerator2
          _)
        (:NewEnumerator1
          (Math|Integer|DecrementInt (Variables|Spawning|GetRemainingForcedTargetMurkyVoleToSpawn)))
        (:NewEnumerator3
          _))
      (Math|Integer|DecrementInt (Variables|Spawning|GetRemainingForcedTargetBarnTowelToSpawn)))
    (:then_2
      ;;[0.9319 0.3559 1] RETURN
      ;; Return the Forced or default variant
      (return (LocalVariables|GetSelectedForceSpawn)))) ; guarantee Return is hit for all exec
)

(fn GetDeterministicSpawnerOrder
  (params (Valid%20EQS%20Spawners :type Array%20of%20Actor%20Object%20References))
  (bind _outrownames (DataTable|GetDataTableRowNames "/BarnTowelArmoryGameplay/DataTables/BarnTowelArmorySpawnLocationData.BarnTowelArmorySpawnLocationData"))
  (LocalVariables|SetRowNames _outrownames)
  ; Find the spawner with the row name matching the deterministic row, and place it in the output array
  (for _array_element (Variables|GetValid%20EQS%20Spawners)
    (bind _asb_barn_towel_spawn_manager (Utilities|Casting|CastToB_BarnTowel_SpawnManager _array_element))
    (bind (_spawndata_datatable _spawndata_rowname) (Utilities|Struct|BreakDataTableRowHandle (Variables|/BarnTowelArmoryGameplay/Items/Environmental/B_BarnTowel_SpawnManager.B_BarnTowel_SpawnManager_C|GetSpawnData_DataTable _asb_barn_towel_spawn_manager)))
    (if (== (Utilities|Array|Get(aref) (LocalVariables|GetRowNames) (+ (Class|FortDeterminismConfig|GetArmoryDeterministicSpawnIndex) (Variables|Spawning|GetDeterministicSpawningDataTableIndexOffset))) _spawndata_rowname)
      (return (Utilities|Array|MakeArray _array_element)))))

(fn InitArmoryVariantForceSpawnVariables
  (bind _effectivemaxarmoryspawns (Variables|Spawning|GetEffectiveMaxArmorySpawns))
  ;[0.4184 0.7107 1] Cache the hotfix values so we don't need to keep getting them from the scalable float
  (bind _output_get (LocalVariables|SethRequestedBarnTowelSpawn (ScalableFloat|GetValueasInteger (Variables|BalanceRows|GetHFRow_SpawningVariantForceBarnTowel))))
  (bind _output_get_2 (LocalVariables|SethRequestedMurkyVoleSpawn (ScalableFloat|GetValueasInteger (Variables|BalanceRows|GetHFRow_SpawningVariantForceMurkyVole))))
  (if (> (+ _output_get_2 _output_get) (Variables|Spawning|GetEffectiveMaxArmorySpawns))
    ;[1 0 0.8896] FALLBACK
    ; We shouldn't exceed the possible spawns, but it may be possible if we request too many guarenteed force spawned and/or the storm is in a strange location
    (if (> (LocalVariables|GethRequestedBarnTowelSpawn) 0)
      (if (> (LocalVariables|GethRequestedMurkyVoleSpawn) 0)
        (switch int _effectivemaxarmoryspawns
          (:Default
            ;;[0.4184 0.7107 1] Each variant has at least one request and we can give each at least 1
            (Variables|Spawning|SetRemainingForcedTargetBarnTowelToSpawn 1)
            (Variables|Spawning|SetRemainingForcedTargetMurkyVoleToSpawn 1)
            (for _ (range (+ (- _effectivemaxarmoryspawns 3) 1))
              (if (<= (Math|Random|RandomIntegerinRange 1 (+ (- (LocalVariables|GethRequestedMurkyVoleSpawn) (Variables|Spawning|GetRemainingForcedTargetMurkyVoleToSpawn)) (- (LocalVariables|GethRequestedBarnTowelSpawn) (Variables|Spawning|GetRemainingForcedTargetBarnTowelToSpawn)))) (- (LocalVariables|GethRequestedBarnTowelSpawn) (Variables|Spawning|GetRemainingForcedTargetBarnTowelToSpawn)))
                (Math|Integer|IncrementInt (Variables|Spawning|GetRemainingForcedTargetBarnTowelToSpawn))
                (else
                  (Math|Integer|IncrementInt (Variables|Spawning|GetRemainingForcedTargetMurkyVoleToSpawn))))))
          (:0
            ;[1 0 0.8896] NO ARMORIES CAN SPAWN
            ; Not expected - but there won't be any Armories spawned. Just in case though, don't use forced spawning at all
            (Fortnite|FortShippingLog self "BARNTOWELARMORY: Cannot find spawn locations to assign variants!" true)
            (Variables|Spawning|SetRemainingForcedTargetBarnTowelToSpawn 0)
            (Variables|Spawning|SetRemainingForcedTargetMurkyVoleToSpawn 0))
          (:1
            ;;[0.4184 0.7107 1] Can only give one spawn - use the requested force spawn as weights.
            (if (<= (Math|Random|RandomIntegerinRange 1 (+ (LocalVariables|GethRequestedMurkyVoleSpawn) (LocalVariables|GethRequestedBarnTowelSpawn))) (LocalVariables|GethRequestedBarnTowelSpawn))
              (Variables|Spawning|SetRemainingForcedTargetBarnTowelToSpawn 1)
              (Variables|Spawning|SetRemainingForcedTargetMurkyVoleToSpawn 0)
              (else
                (Variables|Spawning|SetRemainingForcedTargetBarnTowelToSpawn 0)
                (Variables|Spawning|SetRemainingForcedTargetMurkyVoleToSpawn 1)))))
        (else
          ;;[0.4184 0.7107 1] Give the BarnTowel as many spawns as we can
          ;; (we know it exceeds what is possible, so give all the spawns)
          (Variables|Spawning|SetRemainingForcedTargetBarnTowelToSpawn (Variables|Spawning|GetEffectiveMaxArmorySpawns))))
      (else
        ;;[0.4184 0.7107 1] Give the MurkyVole as many spawns as we can
        ;; (we know it exceeds what is possible, so give all the spawns)
        (Variables|Spawning|SetRemainingForcedTargetMurkyVoleToSpawn (Variables|Spawning|GetEffectiveMaxArmorySpawns))))
    (else
      ;[0 0.03422 1] The hotfix request is meets the current spawning conditions and now special action is needed
      ; The amount of force spawns is less than the total spawns allowed (or there are no forced spawns)
      (Variables|Spawning|SetRemainingForcedTargetBarnTowelToSpawn (LocalVariables|GethRequestedBarnTowelSpawn))
      (Variables|Spawning|SetRemainingForcedTargetMurkyVoleToSpawn (LocalVariables|GethRequestedMurkyVoleSpawn)))))

(fn InitArmoryVariables
  (Variables|Spawning|SetEffectiveMaxArmorySpawns (ScalableFloat|GetValueasInteger (Variables|BalanceRows|GetHFRow_MaxArmoriesToSpawn))))

(fn SelectArmoryFromRegisteredSpawners
  (bind _output_get (LocalVariables|SetSelectedArmorySpawner (Utilities|Array|Get(acopy) (Variables|Spawning|GetArmorySpawnerSelectedOrder) (Utilities|Operators|Subtract (Variables|Spawning|GetEffectiveMaxArmorySpawns) (Variables|Spawning|GetArmoriesSpawnedCount) 1))))
  (CallFunction|SpawnSelectedArmory _output_get))

(fn SetSpawnerAfterEQSQueryComplete
  (params (QueryInstance :type EQS%20Query%20Instance%20Object%20Reference) (QueryStatus :type EEnvQueryStatus%20Enum :default "Processing"))
  (Networking|SwitchHasAuthority
    (:Authority
      (Utilities|IsValid QueryInstance
        (:Is%20Valid
          (bind _resultactors (AI|EQS|GetQueryResultsasActors QueryInstance))
          (if (Class|FortDeterminismConfig|IsUsingArmoryDeterminism)
            (bind _valid_sorted_spawners (Spawning|GetDeterministicSpawnerOrder _resultactors))
            (Variables|Spawning|SetArmorySpawnerSelectedOrder _valid_sorted_spawners)
            (else
              (bind _validsortedspawners (Spawning|CacheandSortSelectedSpawnerOrder _resultactors QueryInstance))
              (Variables|Spawning|SetArmorySpawnerSelectedOrder _validsortedspawners)))
          ;[0.3851 0.6022 1] Set the max number of armories to spawn to the lower of the expected balance Max and the current valid spawn locations
          ; Should generally be able to hit the expected Max Spawns without being limited
          (Variables|Spawning|SetEffectiveMaxArmorySpawns (Math|Integer|Min(Integer) (Variables|Spawning|GetEffectiveMaxArmorySpawns) (Utilities|Array|Length (Variables|Spawning|GetArmorySpawnerSelectedOrder)))))))))
